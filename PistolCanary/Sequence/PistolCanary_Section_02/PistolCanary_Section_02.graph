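# Level Sequence for Section 02 of PistolCanary, coordinating that stretch of the game's timeline.
# A subsequence track, Main A, runs alongside Data Layer tracks for section, day and section-skip ranges, all sharing one end frame.
# Most layer sections open together just before Main A, while S3 and S3_to_MatchEnd begin later, overlapping the section's latter part.
# A possessable countdown Blueprint actor has a visibility track defaulting to false.
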LevelSequence /PistolCanary/Sequence/PistolCanary_Section_02/PistolCanary_Section_02
bindings [PistolCanary_BP_Countdown {
  id=#1
  possessed_class=BP_Generic_Game_Countdown_C
  tracks [{
    type=MovieSceneVisibilityTrack
    sections [{type=MovieSceneVisibilitySection, channels [None {type=MovieSceneScriptingBoolChannel, default=false}]}]
   }]
 }]
display_rate {denominator=1, numerator=60}
tick_resolution {denominator=1, numerator=24000}
tracks [
 {type=MovieSceneSubTrack, display_name=PistolCanary - Main A, sections [{type=MovieSceneSubSection, end_frame=15518, start_frame=3670}]}
 {type=MovieSceneDataLayerTrack, display_name=Data Layer - S1, sections [{type=MovieSceneDataLayerSection, end_frame=15518, start_frame=3600}]}
 {
  type=MovieSceneDataLayerTrack
  display_name=Data Layer - S3_to_MatchEnd
  sections [{type=MovieSceneDataLayerSection, end_frame=15518, start_frame=10193}]
 }
 {
  type=MovieSceneDataLayerTrack
  display_name=Data Layer - S0_to_End_S1
  sections [{type=MovieSceneDataLayerSection, end_frame=15518, start_frame=3600}]
 }
 {
  type=MovieSceneDataLayerTrack
  display_name=Data Layer - Day3_to_End_S1
  sections [{type=MovieSceneDataLayerSection, end_frame=15518, start_frame=3600}]
 }
 {
  type=MovieSceneDataLayerTrack
  display_name="Data Layer - S0_to_End_S3 (Section Skip)"
  sections [{type=MovieSceneDataLayerSection, end_frame=15518, start_frame=3600}]
 }
 {
  type=MovieSceneDataLayerTrack
  display_name=Data Layer - S2_to_MatchEnd
  sections [{type=MovieSceneDataLayerSection, end_frame=15518, start_frame=3600}]
 }
 {
  type=MovieSceneDataLayerTrack
  display_name="Data Layer - S1_to_End_S3 (Section Skip)"
  sections [{type=MovieSceneDataLayerSection, end_frame=15518, start_frame=3600}]
 }
 {type=MovieSceneDataLayerTrack, display_name=Data Layer - S2, sections [{type=MovieSceneDataLayerSection, end_frame=15518, start_frame=3600}]}
 {
  type=MovieSceneDataLayerTrack
  display_name=Data Layer - Day2_to_End_S1
  sections [{type=MovieSceneDataLayerSection, end_frame=15518, start_frame=3600}]
 }
 {
  type=MovieSceneDataLayerTrack
  display_name=Data Layer - S3
  sections [{type=MovieSceneDataLayerSection, end_frame=15518, start_frame=10193}]
 }
]
